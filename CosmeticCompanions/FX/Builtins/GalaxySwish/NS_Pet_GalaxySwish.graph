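# Niagara system for the GalaxySwish quadruped cosmetic pet, adding glows, a sun effect, smoke and a ribbon tail on its skeleton.
# Glows and Sun follow sampled bones and spawn according to System.isBandOn.
# User.NSVariantColor picks among five palettes for Smoke and the active Tail001 ribbon, which follows tail sockets.
# Smoke and ribbon switch local/world position blending at an owner-speed threshold; the older Tail emitter is disabled.
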NiagaraSystem /CosmeticCompanions/FX/Builtins/GalaxySwish/NS_Pet_GalaxySwish
dependencies {
 usedDataInterfaces [
  &NiagaraDataInterfaceArrayInt32
  &NiagaraDataInterfaceCurve
  &NiagaraDataInterfaceColorCurve
  &NiagaraDataInterfaceSkeletalMesh
 ]
 usedDynamicInputs []
 usedModules [
  &self:SystemSpawnScript.SystemScriptSource.SystemScriptGraph.NiagaraNodeAssignment_0.SetVariables_#1
  &self:SystemSpawnScript.SystemScriptSource.SystemScriptGraph.NiagaraNodeAssignment_1.SetVariables_#2
  &SystemState
  &BlendSpaceHelper
 ]
}
emitters {
 Glows {
  emitterUpdateScript [
   &EmitterState(Life Cycle Mode=::NewEnumerator1, Loop Behavior=::NewEnumerator1, Loop Duration Mode=::NewEnumerator1, UseLoopDelay=false)
   &SpawnBurst_Instantaneous(Spawn Count=7, Spawn Probability=$System.isBandOn, Use Spawn Probability=true, UseLoopCountLimit=false) {module=SpawnBurst_Instantaneous001}
  ]
  particleSpawnScript [
   &FN_SetSpawnGroup(Number of Spawn Groups=7)
   /SetVariables(Particles.VisibilityTag=&SpawnGroup()) {isSetParameters=true}
   &InitializeParticle(
    Color=&MergeHelper_ColorSwitch(
     0=<rgba 8 1.7733 0 0.05>
     1=<rgba 3 1.3919 0 0.125>
     2=<rgba 3 1.3919 0 0.175>
     3=<rgba 3 1.3919 0 0.175>
     4=<rgba 3 1.3919 0 0.175>
     5=<rgba 3 1.3919 0 0.175>
     Index=&SpawnGroup()
     Num=::NewEnumerator4
    )
    Mesh Scale=<0.2525 0.2525 0.625>
    Mesh Scale Mode=::NewEnumerator4
    Sprite Size Mode=::NewEnumerator3
    Uniform Sprite Size=&MergeHelper_FloatSwitch(0=0, 1=18, 2=14, 3=14, 4=14, 5=14, Index=&SpawnGroup(), Num=::NewEnumerator4)
    UsePositionOffset=false
    Write Material Random=true
   )
   &CameraOffset(Camera Offset Amount=12)
   &InitialMeshOrientation(Mesh Orientation Mode=::NewEnumerator1, Use Rotation Vector=true) {enabled=false}
   /SetVariables(Particles.NeckCO=13) {isSetParameters=true}
  ]
  particleUpdateScript [
   &ParticleState(Kill Particles When Lifetime Has Elapsed=false, Let Infinitely Lived Particles Die When Emitter Deactivates=true)
   /SetVariables(
    Particles.2Color:LinearColor=<rgba 24 3.4026 1.3368 1>
    Particles.2DP:Vector4f=&MakeVector4(X=0, Y=0, Z=0, W=0)
    Particles.2Scale:Vector3f=&VectorFromFloat(Value=0.0425)
   ) {isSetParameters=true}
   &SkeletalMeshLocation(
    Bone / Socket Index=&SelectIntFromArray(
     Direct Array Index=&SpawnGroup()
     Int Selection Array=DataInterface() {propertyValues {IntData [0,1,2,3,4,5,9]}}
    )
    Bone Sampling Mode=::NewEnumerator4
    MaskSpawnGroup=false
    Sampled Position Offset=&MergeHelper_VectorSwitch(
     0=<0 -1 0>
     1=<0 0 0>
     2=<3 0 0>
     3=<-3 0 0>
     4=<12 0 0>
     5=<-12 0 0>
     6=<0 0 0>
     Index=&SpawnGroup()
     Num=::NewEnumerator5
    )
    Skeletal Mesh=$System.Skeletal Mesh Reader
   )
  ]
  rendererClasses [&NiagaraSpriteRendererProperties, &NiagaraMeshRendererProperties]
  renderers [
   {rendererClass=&NiagaraSpriteRendererProperties}
   {rendererClass=&NiagaraMeshRendererProperties, rendererIndex=1}
   {rendererClass=&NiagaraMeshRendererProperties, rendererIndex=2}
   {rendererClass=&NiagaraSpriteRendererProperties, rendererIndex=3}
   {rendererClass=&NiagaraSpriteRendererProperties, rendererIndex=4}
   {rendererClass=&NiagaraSpriteRendererProperties, rendererIndex=5}
   {rendererClass=&NiagaraSpriteRendererProperties, rendererIndex=6}
  ]
 }
 Smoke {
  emitterUpdateScript [
   &EmitterState(Life Cycle Mode=::NewEnumerator1, Loop Behavior=::NewEnumerator1, Loop Duration Mode=::NewEnumerator1, UseLoopDelay=false)
   &SpawnRate(SpawnRate=20, Use Spawn Probability=false)
   &SpawnRate(SpawnGroup=1, SpawnRate=56, Use Spawn Probability=false) {module=SpawnRate001}
  ]
  particleSpawnScript [
   /SetVariables(Particles.VisibilityTag=&SpawnGroup()) {isSetParameters=true}
   &InitializeParticle(
    Color=&MultiplyLinearColor(
     Linear Color=<rgba 1 1 1 0.8>
     Scale Factor=&Multiply_Vector4(
      A=&MakeVector4_FromLinearColor(LinearColor=&MaskLinearColorBySpawnGroup(Masked Linear Color=<rgba 3 3 3 0.01>, Passthrough Linear Color=<rgba 1 1 1 1>))
      B=&MakeVector4_FromLinearColor(LinearColor=&MergeHelper_ColorSwitch(0=<rgba 1 1 1 1>, 1=<rgba 1 1 1 0>, 2=<rgba 1 1 1 1>, 3=<rgba 1 1 1 1>, 4=<rgba 0.3095 0.7231 1 1>, Index=0, Num=::NewEnumerator3))
     )
    )
    Lifetime=&MaskFloatBySpawnGroup(
     Masked Float=&RandomRangeFloat(Minimum=1.2, Maximum=1.8)
     Passthrough Float=&RandomRangeFloat(Minimum=1.2, Maximum=1.6)
    )
    Sprite Rotation Mode=::NewEnumerator2
    Sprite Size Mode=::NewEnumerator3
    Uniform Sprite Size=&MaskFloatBySpawnGroup(
     Masked Float=&RandomRangeFloat(Minimum=24, Maximum=32)
     Passthrough Float=&RandomRangeFloat(Minimum=32, Maximum=36)
    )
    UsePositionOffset=false
    Write Material Random=true
   )
   &SkeletalMeshLocation(
    Bone / Socket Index=&RandomRangeInt(Minimum=6, Maximum=7)
    Bone Sampling Mode=::NewEnumerator4
    MaskSpawnGroup=false
    Skeletal Mesh=$System.Skeletal Mesh Reader
   )
   &ShapeLocation(
    MaskSpawnGroup=false
    Non Uniform Scale=<1 1.4 1>
    Offset=<0 3 4>
    Offset Mode=::NewEnumerator0
    Override Seed=false
    Shape Origin=$Particles.SkeletalMeshLocation.SampledPosition
    Sphere Radius=6
    UseOwnerScale=true
   )
   &AddVelocity(
    Cone Angle=25
    Cone Axis=&RotateVectorByQuaternion(Vector To Rotate=<1 0 0>, Quaternion=$Particles.SkeletalMeshLocation.SampledRotation)
    Module.UseConeFalloff=false
    Module.UseVelDistribution=false
    Override Seed=false
    Velocity Mode=::NewEnumerator2
    Velocity Speed=&RandomRangeFloat(Minimum=25, Maximum=32)
   )
   &CameraOffset(Camera Offset Amount=&RandomRangeFloat(Minimum=0, Maximum=8))
  ]
  particleUpdateScript [
   &ParticleState()
   &ScaleSpriteSize(
    Uniform Curve Scale=1.2
    Uniform Curve Sprite Scale=DataInterface() {propertyValues {Curve {
       defaultValue=0
       keys [
        {arriveTangent=1.2454, interpMode=RCIM_Cubic, leaveTangent=1.2454, tangentMode=RCTM_User, time=0, value=0.4}
        {arriveTangent=0.0224, interpMode=RCIM_Cubic, leaveTangent=0.0224, tangentMode=RCTM_User, time=1, value=1}
       ]
      }}}
   )
   &CurlNoiseForce(Bias Noise=false, Noise Frequency=20, Pan Noise=false)
   &SolveForcesAndVelocity(Clamp Velocity=false, Limit Acceleration=false)
   &ScaleColor(
    Scale Alpha=&FloatFromCurve(
     FloatCurve=DataInterface() {propertyValues {Curve {
        defaultValue=1
        keys [
         {arriveTangent=9.0725, interpMode=RCIM_Cubic, leaveTangent=9.0725, tangentMode=RCTM_User, time=0, value=0}
         {arriveTangent=-2.5249, interpMode=RCIM_Cubic, leaveTangent=-2.5249, tangentMode=RCTM_User, time=0.1408, value=0.9847}
         {arriveTangent=-0.0405, interpMode=RCIM_Cubic, leaveTangent=-0.0405, tangentMode=RCTM_User, time=1, value=0}
        ]
       }}}
     Scale Curve=0.6
    )
    Scale RGB=&MakeVectorFromLinearColorRGB(Color=&MergeHelper_ColorSwitch(
      0=&ColorFromCurve(ColorCurve=DataInterface() {propertyValues {
         AlphaCurve {keys []}
         BlueCurve {keys [<at 0 1 RCIM_Cubic>,<at 0.7 1>]}
         GreenCurve {keys [<at 0 0.6554 RCIM_Cubic>,<at 0.7 0.0744>]}
         RedCurve {keys [<at 0 0.1542 RCIM_Cubic>,<at 0.7 0.0625>]}
        }})
      1=&ColorFromCurve(ColorCurve=DataInterface() {propertyValues {
         AlphaCurve {keys []}
         BlueCurve {keys [<at 0 1 RCIM_Cubic>,<at 0.7 1>]}
         GreenCurve {keys [<at 0 0.1779 RCIM_Cubic>,<at 0.7 0>]}
         RedCurve {keys [<at 0 0.9473 RCIM_Cubic>,<at 0.7 0.2988>]}
        }})
      2=&ColorFromCurve(ColorCurve=DataInterface() {propertyValues {
         AlphaCurve {keys [<at 0 1 RCIM_Cubic>]}
         BlueCurve {keys [<at 0 0 RCIM_Cubic>,<at 0.7 0.1185 RCIM_Cubic>]}
         GreenCurve {keys [<at 0 0.3467 RCIM_Cubic>,<at 0.7 0 RCIM_Cubic>]}
         RedCurve {keys [<at 0 1 RCIM_Cubic>,<at 0.7 1 RCIM_Cubic>]}
        }})
      3=&ColorFromCurve(ColorCurve=DataInterface() {propertyValues {
         AlphaCurve {keys [<at 0 1 RCIM_Cubic>]}
         BlueCurve {keys [<at 0 0.9387 RCIM_Cubic>,<at 0.7 1 RCIM_Cubic>]}
         GreenCurve {keys [<at 0 1 RCIM_Cubic>,<at 0.7 0.2379 RCIM_Cubic>]}
         RedCurve {keys [<at 0 0.0284 RCIM_Cubic>,<at 0.7 0 RCIM_Cubic>]}
        }})
      4=&ColorFromCurve(ColorCurve=DataInterface() {propertyValues {
         AlphaCurve {keys [<at 0 1 RCIM_Cubic>]}
         BlueCurve {keys [<at 0 1 RCIM_Cubic>,<at 0.7 1 RCIM_Cubic>]}
         GreenCurve {keys [<at 0 0.8171 RCIM_Cubic>,<at 0.7 0.6072 RCIM_Cubic>]}
         RedCurve {keys [<at 0 0.1193 RCIM_Cubic>,<at 0.7 0.0014 RCIM_Cubic>]}
        }})
      Index=&MakeIntFromFloat(Float=$User.NSVariantColor)
      Num=::NewEnumerator3
     ))
    ScaleA=true
    ScaleRGB=true
   )
   &DynamicMaterialParameters(
    Index 0 Param 1=1
    Index 0 Param 3=1
    Module.Index0Float0Write=true
    Module.Index0Float1Write=true
    Module.Index0Float2Write=true
    Module.Index0Float3Write=true
   )
   /Game/Effects/Niagara/Modules/Sprite/FN_SpriteUV(Write to Module Namepsace=true)
   /SetVariables() {isSetParameters=true}
   &BlendSpace(
    "Position Blend (World <-> Local)":NiagaraFloat=&MakeCustomFloatFromBool(
     Bool=&SetBoolByFloatComparison(A=&VectorLength(Vector=$Engine.Owner.Velocity), B=300)
     True Float=0
     False Float=1
     Translation Contribution:NiagaraFloat=1
     Rotation Contribution:NiagaraFloat=1
    )
    Position Coordinate Space=1
   )
  ]
  rendererClasses [&NiagaraSpriteRendererProperties]
  renderers [
   {rendererClass=&NiagaraSpriteRendererProperties}
   {rendererClass=&NiagaraSpriteRendererProperties, rendererIndex=1}
  ]
 }
 Sun {
  emitterUpdateScript [
   &EmitterState(Life Cycle Mode=::NewEnumerator1, Loop Behavior=::NewEnumerator1, Loop Duration Mode=::NewEnumerator1, UseLoopDelay=false)
   &SpawnBurst_Instantaneous(Spawn Count=5, Spawn Probability=$System.isBandOn, Use Spawn Probability=true, UseLoopCountLimit=false)
   &SpawnRate(Spawn Probability=$System.isBandOn, SpawnGroup=1, SpawnRate=2, Use Spawn Probability=true) {module=SpawnRate001}
   &SpawnRate(Spawn Probability=$System.isBandOn, SpawnRate=8, Use Spawn Probability=true)
  ]
  particleSpawnScript [
   /SetVariables(Particles.VisibilityTag=&SpawnGroup()) {isSetParameters=true}
   &InitializeParticle(
    Color=&MergeHelper_ColorSwitch(
     0=&RandomRangeLinearColor(Minimum=<rgba 2 0.6993 0.0962 0.275>, Maximum=<rgba 1 0.6177 0.0677 0.1>)
     1=<rgba 8 4.0826 0.6635 0.6>
     2=<rgba 0 0 0 0>
     3=<rgba 0 0 0 0>
     4=<rgba 0 0 0 0>
     Index=&SpawnGroup()
     Num=::NewEnumerator3
    )
    Lifetime=&MaskFloatBySpawnGroup(
     Masked Float=&RandomRangeFloat(Minimum=3, Maximum=2.4)
     Passthrough Float=&RandomRangeFloat(Minimum=1, Maximum=1.4)
    )
    Mesh Scale=&MergeHelper_VectorSwitch(
     0=&Multiply_VectorByFloat(Vector=&RandomRangeVector(Minimum=<0.225 0.2 0.08>, Maximum=<0.125 0.5 0.1>), Float=0.65)
     1=&VectorFromFloat(Value=&Multiply_Float(A=&RandomRangeFloat(Minimum=0.025, Maximum=0.04), B=0.05))
     2=<0 0 0>
     3=<0 0 0>
     4=<0 0 0>
     Index=&SpawnGroup()
     Num=::NewEnumerator3
    )
    Mesh Scale Mode=::NewEnumerator4
    Sprite Size Max=<3 8>
    Sprite Size Min=<2 5>
    Sprite Size Mode=::NewEnumerator2
    UsePositionOffset=false
    Write Material Random=true
   )
   &ShapeLocation(
    MaskSpawnGroup=false
    Override Seed=false
    Sphere Radius=&MergeHelper_FloatSwitch(0=1, 1=0, 2=0, 3=0, 4=0, Index=&SpawnGroup(), Num=::NewEnumerator3)
    Sphere Surface Distribution=1
    UseOwnerScale=true
   )
   &AddVelocity(
    Module.Use Falloff Distance=false
    Override Seed=false
    Velocity Mode=::NewEnumerator1
    Velocity Speed=&MergeHelper_FloatSwitch(
     0=&RandomRangeFloat(Minimum=5, Maximum=15)
     1=0
     2=0
     3=0
     4=0
     Index=&SpawnGroup()
     Num=::NewEnumerator3
    )
   ) {enabled=false}
   &InitialMeshOrientation(Mesh Orientation Mode=::NewEnumerator2, Orientation Vector=$Particles.ShapeLocation.ShapeNormal, Use Rotation Vector=true)
   &CameraOffset(Camera Offset Amount=&MergeHelper_FloatSwitch(0=0, 1=15, 2=0, 3=0, 4=0, Index=&SpawnGroup(), Num=::NewEnumerator3))
  ]
  particleUpdateScript [
   &ParticleState()
   &ScaleMeshSize(Scale Factor=&VectorFromFloat(Value=&FloatFromCurve(FloatCurve=DataInterface() {propertyValues {Curve {
         defaultValue=1
         keys [
          {arriveTangent=4.1785, interpMode=RCIM_Cubic, leaveTangent=4.1785, tangentMode=RCTM_User, time=0, value=0}
          {arriveTangent=1.9258, interpMode=RCIM_Cubic, leaveTangent=1.9258, tangentMode=RCTM_User, time=0.2244, value=0.8364}
          {arriveTangent=0.0224, interpMode=RCIM_Cubic, leaveTangent=0.0224, tangentMode=RCTM_User, time=1, value=1.45}
         ]
        }}})))
   &SkeletalMeshLocation(
    Bone / Socket Index=1
    Bone Sampling Mode=::NewEnumerator4
    MaskSpawnGroup=false
    Position Sampling=::NewEnumerator1
    Sampled Position Offset:NiagaraPosition=$Particles.ShapeLocation.ShapePosition
    Skeletal Mesh=$System.Skeletal Mesh Reader
   )
   &ScaleSpriteSize(Uniform Curve Sprite Scale=DataInterface() {propertyValues {Curve {
       defaultValue=0
       keys [
        {interpMode=RCIM_Cubic, tangentMode=RCTM_User, time=0, value=0}
        {interpMode=RCIM_Cubic, tangentMode=RCTM_User, time=0.5, value=1}
        {tangentMode=RCTM_User, time=1, value=0}
       ]
      }}})
   &SolveForcesAndVelocity(Clamp Velocity=false, Limit Acceleration=false)
   &ScaleColor(
    Scale Alpha=&MergeHelper_FloatSwitch(
     0=&FloatFromCurve(FloatCurve=DataInterface() {propertyValues {Curve {
         defaultValue=1
         keys [
          {tangentMode=RCTM_User, time=0, value=0}
          <at 0.6 1>
          {arriveTangent=-0.0405, interpMode=RCIM_Cubic, leaveTangent=-0.0405, tangentMode=RCTM_User, time=0.7009, value=1}
          {arriveTangent=-8.0089, interpMode=RCIM_Cubic, leaveTangent=-8.0089, tangentMode=RCTM_User, time=1, value=0}
         ]
        }}})
     1=&FloatFromCurve(
      FloatCurve=DataInterface() {propertyValues {Curve {
         defaultValue=1
         keys [
          {arriveTangent=9.0725, interpMode=RCIM_Cubic, leaveTangent=9.0725, tangentMode=RCTM_User, time=0, value=0}
          {arriveTangent=-2.5249, interpMode=RCIM_Cubic, leaveTangent=-2.5249, tangentMode=RCTM_User, time=0.1408, value=0.9847}
          {arriveTangent=-0.0405, interpMode=RCIM_Cubic, leaveTangent=-0.0405, tangentMode=RCTM_User, time=1, value=0}
         ]
        }}}
      Scale Curve=&RandomRangeFloat(Minimum=0.5, Maximum=0.8)
     )
     2=0
     3=0
     4=0
     Index=&SpawnGroup()
     Num=::NewEnumerator3
    )
    ScaleA=true
    ScaleRGB=true
   )
  ]
  rendererClasses [&NiagaraMeshRendererProperties]
  renderers [
   {rendererClass=&NiagaraMeshRendererProperties, rendererIndex=0}
   {rendererClass=&NiagaraMeshRendererProperties, rendererIndex=1}
  ]
 }
 Tail {
  bEnabled=false
  emitterUpdateScript [
   &EmitterState(Life Cycle Mode=::NewEnumerator1, Loop Behavior=::NewEnumerator1, Loop Duration Mode=::NewEnumerator1, UseLoopDelay=false)
   &SpawnRate(SpawnRate=60, Use Spawn Probability=false)
   /SetVariables(
    Emitter.CenterColor:LinearColor=<rgba 0.75 0.9574 1 1>
    Emitter.ColorLerp1:LinearColor=<rgba 0.0466 0.0195 1 1>
    Emitter.ColorLerp2:LinearColor=<rgba 0.0412 1.0276 2 1>
    Emitter.MainColor:LinearColor=<rgba 0.5898 1.3603 2 1>
   ) {isSetParameters=true, module=SetVariables_#3}
  ]
  particleSpawnScript [
   &InitializeParticle(
    Lifetime=&MakeCustomFloatFromBool(
     Bool=&SetBoolByFloatComparison(A=&VectorLength(Vector=$Engine.Owner.Velocity), B=300)
     True Float=0.135
     False Float=0.17
    )
    Ribbon Facing Mode=::NewEnumerator1
    Ribbon Facing Vector:Vector3f=$Engine.Owner.SystemYAxis
    Ribbon Width=10
    Ribbon Width Mode=::NewEnumerator1
    UsePositionOffset=false
    Write Material Random=true
   )
   &SkeletalMeshLocation(
    Bone / Socket Index=6
    Bone Sampling Mode=::NewEnumerator4
    MaskSpawnGroup=false
    Orientation Sampling=::NewEnumerator1
    Sampled Position Offset=<3 0 0>
    Skeletal Mesh=$System.Skeletal Mesh Reader
   )
   /SetVariables(Particles.RibbonFacing1:Vector3f=$Engine.Owner.SystemXAxis) {isSetParameters=true, module=SetVariables_#4}
   &AddVelocity(Velocity=&MakeCustomVectorFromBool(
     Bool=&SetBoolByFloatComparison(A=&VectorLength(Vector=$Engine.Owner.Velocity), B=300)
     True Vector=&Multiply_Vector(
      A=&VectorFromFloat(Value=30)
      B=&RotateVectorByQuaternion(Vector To Rotate=<1 0 0>, Quaternion=$Particles.SkeletalMeshLocation.SampledRotation)
     )
     False Vector=&Multiply_Vector(
      A=&VectorFromFloat(Value=120)
      B=&RotateVectorByQuaternion(Vector To Rotate=<1 0 0>, Quaternion=$Particles.SkeletalMeshLocation.SampledRotation)
     )
    ))
   &CameraOffset(Camera Offset Amount=5)
  ]
  particleUpdateScript [
   &ParticleState()
   &SolveForcesAndVelocity(Clamp Velocity=false, Limit Acceleration=false)
   &ScaleRibbonWidth(Ribbon Width Scale=&FloatFromCurve(FloatCurve=DataInterface() {propertyValues {Curve {
        defaultValue=1
        keys [
         {interpMode=RCIM_Cubic, tangentMode=RCTM_User, time=0.125, value=1}
         {arriveTangent=-0.0799, interpMode=RCIM_Cubic, leaveTangent=-0.0799, tangentMode=RCTM_User, time=1, value=0}
        ]
       }}}))
   &BlendSpace(
    "Position Blend (World <-> Local)":NiagaraFloat=&MakeCustomFloatFromBool(
     Bool=&SetBoolByFloatComparison(A=&VectorLength(Vector=$Engine.Owner.Velocity), B=300)
     True Float=0
     False Float=1
     Translation Contribution:NiagaraFloat=1
     Rotation Contribution:NiagaraFloat=1
    )
    Position Coordinate Space=1
   )
   &ScaleColor(
    Scale Alpha=&FloatFromCurve(
     FloatCurve=DataInterface() {propertyValues {Curve {keys [
         {arriveTangent=-0.2102, interpMode=RCIM_Cubic, leaveTangent=-0.2102, tangentMode=RCTM_User, time=0, value=1}
         {arriveTangent=-1.1728, interpMode=RCIM_Cubic, leaveTangent=-1.1728, tangentMode=RCTM_User, time=1, value=0}
        ]}}}
     Scale Curve=0.525
    )
    ScaleA=true
    ScaleRGB=true
   )
   /SetVariables(
    Particles.Color1:LinearColor=&MultiplyLinearColor(Linear Color=$Particles.Color, Scale Factor=&MakeVector4_FromFloat(Float=1))
    Particles.RibbonWidth1:NiagaraFloat=&FloatFromCurve(
     FloatCurve=DataInterface() {propertyValues {Curve {
        defaultValue=1
        keys [
         {interpMode=RCIM_Cubic, tangentMode=RCTM_User, time=0, value=1}
         {arriveTangent=-2.485, interpMode=RCIM_Cubic, leaveTangent=-2.485, tangentMode=RCTM_User, time=1, value=0}
        ]
       }}}
     Scale Curve=10
    )
   ) {isSetParameters=true, module=SetVariables_#5}
  ]
  rendererClasses [&NiagaraRibbonRendererProperties]
  renderers [
   {rendererClass=&NiagaraRibbonRendererProperties}
   {rendererClass=&NiagaraRibbonRendererProperties, rendererIndex=1}
  ]
 }
 Tail001 {
  emitterUpdateScript [
   &EmitterState(Life Cycle Mode=::NewEnumerator1, Loop Behavior=::NewEnumerator1, Loop Duration Mode=::NewEnumerator1, UseLoopDelay=false)
   &SpawnRate(SpawnRate=60, Use Spawn Probability=false)
   /SetVariables(
    Emitter.CenterColor:LinearColor=&MergeHelper_ColorSwitch(
     0=<rgba 0.75 0.9574 1 1>
     1=<rgba 0.8587 0.75 1 1>
     2=<rgba 1 0.8022 0.75 1>
     3=<rgba 0.75 1 0.9804 1>
     4=<rgba 1 0.75 0.8413 1>
     Index=&MakeIntFromFloat(Float=$User.NSVariantColor)
     Num=::NewEnumerator3
    )
    Emitter.ColorLerp1:LinearColor=&MergeHelper_ColorSwitch(
     0=<rgba 0.0466 0.0195 1 1>
     1=<rgba 0.529 0.0198 1 1>
     2=<rgba 1 0.3147 0 1>
     3=<rgba 0.0048 1 0.3516 1>
     4=<rgba 0 0.7503 1 1>
     Index=&MakeIntFromFloat(Float=$User.NSVariantColor)
     Num=::NewEnumerator3
    )
    Emitter.ColorLerp2:LinearColor=&MergeHelper_ColorSwitch(
     0=<rgba 0.0412 1.0276 2 1>
     1=<rgba 1.1528 0.0763 2 1>
     2=<rgba 2 1.1069 0 1>
     3=<rgba 0 1.1597 2 1>
     4=<rgba 0.0412 1.0276 2 1>
     Index=&MakeIntFromFloat(Float=$User.NSVariantColor)
     Num=::NewEnumerator3
    )
    Emitter.MainColor:LinearColor=&MergeHelper_ColorSwitch(
     0=<rgba 0.5898 1.3603 2 1>
     1=<rgba 1.2029 0.5898 2 1>
     2=<rgba 2 0.737 0.5898 1>
     3=<rgba 0.5898 2 1.7425 1>
     4=<rgba 0.5898 1.3603 2 1>
     Index=&MakeIntFromFloat(Float=$User.NSVariantColor)
     Num=::NewEnumerator3
    )
   ) {isSetParameters=true, module=SetVariables_#3}
  ]
  particleSpawnScript [
   &InitializeParticle(
    Lifetime=&MakeCustomFloatFromBool(
     Bool=&SetBoolByFloatComparison(A=&VectorLength(Vector=$Engine.Owner.Velocity), B=300)
     True Float=0.2
     False Float=0.185
    )
    Ribbon Facing Mode=::NewEnumerator1
    Ribbon Facing Vector:Vector3f=$Engine.Owner.SystemYAxis
    Ribbon Width=10
    Ribbon Width Mode=::NewEnumerator1
    UsePositionOffset=false
    Write Material Random=true
   )
   &SkeletalMeshLocation(
    Bone / Socket Index=6
    Bone Sampling Mode=::NewEnumerator4
    MaskSpawnGroup=false
    Orientation Sampling=::NewEnumerator1
    Position Sampling=::NewEnumerator2
    Sampled Position Offset=<3 0 0>
    Skeletal Mesh=$System.Skeletal Mesh Reader
   )
   /SetVariables(Particles.RibbonFacing1:Vector3f=$Engine.Owner.SystemXAxis) {isSetParameters=true, module=SetVariables_#4}
   &CameraOffset(Camera Offset Amount=5)
   &AddVelocity(Velocity=&MakeCustomVectorFromBool(
     Bool=&SetBoolByFloatComparison(A=&VectorLength(Vector=$Engine.Owner.Velocity), B=300)
     True Vector=&Multiply_Vector(
      A=&VectorFromFloat(Value=625)
      B=&RotateVectorByQuaternion(Vector To Rotate=<1 0 0>, Quaternion=$Particles.SkeletalMeshLocation.SampledRotation)
     )
     False Vector=&Multiply_Vector(
      A=&VectorFromFloat(Value=500)
      B=&RotateVectorByQuaternion(Vector To Rotate=<1 0 0>, Quaternion=$Particles.SkeletalMeshLocation.SampledRotation)
     )
    ))
  ]
  particleUpdateScript [
   &ParticleState()
   &ScaleVelocity(Velocity Scale=&VectorFromFloat(Value=&FloatFromCurve(FloatCurve=DataInterface() {propertyValues {Curve {keys [<at 0 0>,<at 0.7 0>,<at 0.8 1>,<at 1 1>]}}})))
   &SolveForcesAndVelocity(Clamp Velocity=false, Limit Acceleration=false)
   &ScaleRibbonWidth(Ribbon Width Scale=&FloatFromCurve(FloatCurve=DataInterface() {propertyValues {Curve {
        defaultValue=1
        keys [
         {arriveTangent=8.0029, interpMode=RCIM_Cubic, leaveTangent=8.0029, tangentMode=RCTM_User, time=0, value=0}
         {interpMode=RCIM_Cubic, tangentMode=RCTM_User, time=0.125, value=1}
        ]
       }}}))
   &BlendSpace(
    "Position Blend (World <-> Local)":NiagaraFloat=&MakeCustomFloatFromBool(
     Bool=&SetBoolByFloatComparison(A=&VectorLength(Vector=$Engine.Owner.Velocity), B=300)
     True Float=0
     False Float=1
     Translation Contribution:NiagaraFloat=1
     Rotation Contribution:NiagaraFloat=1
    )
    Position Coordinate Space=1
   )
   &ScaleColor(
    Scale Alpha=&FloatFromCurve(
     FloatCurve=DataInterface() {propertyValues {Curve {keys [
         {arriveTangent=-0.2102, interpMode=RCIM_Cubic, leaveTangent=-0.2102, tangentMode=RCTM_User, time=0, value=1}
         {arriveTangent=-1.1728, interpMode=RCIM_Cubic, leaveTangent=-1.1728, tangentMode=RCTM_User, time=1, value=0}
        ]}}}
     Scale Curve=0.525
    )
    ScaleA=true
    ScaleRGB=true
   )
   /SetVariables(
    Particles.Color1:LinearColor=&MultiplyLinearColor(Linear Color=$Particles.Color, Scale Factor=&MakeVector4_FromFloat(Float=1))
    Particles.RibbonWidth1:NiagaraFloat=&FloatFromCurve(
     FloatCurve=DataInterface() {propertyValues {Curve {
        defaultValue=1
        keys [
         {interpMode=RCIM_Cubic, tangentMode=RCTM_User, time=0, value=1}
         {arriveTangent=-2.485, interpMode=RCIM_Cubic, leaveTangent=-2.485, tangentMode=RCTM_User, time=1, value=0}
        ]
       }}}
     Scale Curve=10
    )
   ) {isSetParameters=true, module=SetVariables_#5}
   /Game/Effects/Niagara/Modules/Curves/FN_ListCurve_SkeletalMeshSampler(
    Curve Index:NiagaraFloat=&FloatFromCurve(FloatCurve=DataInterface() {propertyValues {Curve {defaultValue=1, keys [{tangentMode=RCTM_User, time=0, value=0}, {tangentMode=RCTM_User, time=0.8, value=1}]}}})
    Index_02=1
    Index_03=2
    Number of Inputs=1
    Selection Type=0
    Skeletal Mesh=DataInterface() {propertyValues {
      DefaultMesh=null
      ExcludeBoneName=root
      OptionalComponentClass=null
      PreviewMesh=/CosmeticCompanions/Assets/Quadruped/GalaxySwish/Meshes/GalaxySwish_Quadruped
      SkinningMode=SkinOnTheFly
      SoftSourceActor=null
      SourceMode=Default
      UvSetIndex=0
      WholeMeshLOD=-1
      bExcludeBone=false
      bReadDeformedGeometry=true
      bRequireCurrentFrameData=true
      ComponentTags []
      FilteredBones []
      FilteredSockets [tail_03Socket, tail_04Socket_0, tail_04Socket]
      MeshUserParameter {parameter {name=None, typeDefHandle {registeredTypeIndex=1}, varData []}}
      SamplingRegions []
     }}
   )
   /SetVariables(Particles.Position=&AddVectorToPosition(
     Position=$Output.FN_ListCurve_SkeletalMeshSampler.Position
     Vector=&Multiply_VectorByFloat(Vector=$Particles.Velocity, Float=$Engine.DeltaTime)
    )) {isSetParameters=true}
  ]
  rendererClasses [&NiagaraRibbonRendererProperties]
  renderers [
   {rendererClass=&NiagaraRibbonRendererProperties}
   {rendererClass=&NiagaraRibbonRendererProperties, rendererIndex=1}
  ]
 }
}
summary {
 systemName=NS_Pet_GalaxySwish
 emitters [
  {emitterName=Glows, rendererClasses [&NiagaraSpriteRendererProperties, &NiagaraMeshRendererProperties]}
  {emitterName=Sun, rendererClasses [&NiagaraMeshRendererProperties]}
  {emitterName=Smoke, rendererClasses [&NiagaraSpriteRendererProperties]}
  {bEnabled=false, emitterName=Tail, rendererClasses [&NiagaraRibbonRendererProperties]}
  {emitterName=Tail001, rendererClasses [&NiagaraRibbonRendererProperties]}
 ]
 userVariables [
  User.NSVariantColor {type=NiagaraFloat, defaultValue=4, description=""}
  User.RibbonIDOffset {type=NiagaraInt32}
  User.VariantBand {type=NiagaraFloat, defaultValue=1, description=""}
 ]
}
